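# Translucent unlit Niagara material for ribbons and mesh particles, drawing flowing spline energy for the Demon absorb portal effect.
# World position offset adds SplineThicken to a clamped, Niagara-driven displacement, weighted along U by a power curve.
# Emissive layers panning plasma, smoke and desaturated rainbow noise, distorted by inky smoke and tinted by particle color.
# Opacity combines smoke, sine end falloff, depth and view-angle fades, an optional appear mask and a particle-centered sphere mask.
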Material /GrillMoon/Assets/FX/Demon/Absorb/M_GrillMoon_SplinePortalEnergy
graph {
 properties {
  bUsedWithNiagaraMeshParticles=true
  bUsedWithNiagaraRibbons=true
  blendMode=BLEND_Translucent
  pixelDepthOffsetMode=PDOM_Legacy
  shadingModel=MSM_Unlit
  twoSided=true
  usedShadingModels=Unlit
 }
 nodes [
  n1 = ScalarParameter() {defaultValue=25, parameterName=Width}
  n2 = DynamicParameter() {desc=Parameters set by Niagara, parameterIndex=1, paramNames [X,Y,Z,WPOIntensity]}
  n3 = ParticlePositionWS()
  n4 = DynamicParameter() {desc=Parameters set by Niagara, parameterIndex=2, paramNames [X, Y, Z, ""]}
  n5 = MaterialFunctionCall(X=@n4.X, Y=@n4.X, Z=@n4.X) {materialFunction=&MakeFloat3}
  n6 = ScalarParameter() {parameterName=FEPositionFix}
  n7 = ScalarParameter() {defaultValue=100, parameterName=Interpolate_WPO_Range}
  n8 = Clamp(
   None=Subtract(
    A=MaterialFunctionCall(X=@n2.X, Y=@n2.X, Z=@n2.X) {materialFunction=&MakeFloat3}.Result
    B=LinearInterpolate(A=@n3, B=@n5.Result, Alpha=@n6)
   )
   Min=Multiply(A=@n7) {constB=-1}
   Max=@n7
  ) {maxDefault=50, minDefault=-50}
  n9 = TextureCoordinate()
  n10 = Multiply(A=Power(Base=ComponentMask(None=@n9) {r=true}, Exp=ScalarParameter() {defaultValue=2.5, parameterName=WPOBendPower}), B=@n2.WPOIntensity)
  Add(
   A=MaterialFunctionCall(WidthBase=@n1, WidthTip=@n1, Expand U or V UV Channel=StaticBoolParameter() {parameterName="Expand U or V UV Channel (SB)"}) {materialFunction=&SplineThicken}.WorldPositionOffset
   B=LinearInterpolate(A=Constant(), B=LinearInterpolate(B=@n8, Alpha=@n10), Alpha=ScalarParameter() {defaultValue=1, parameterName=Use interpolate WPO})
  ) {constB=0}
  Constant() {r=20}
  n11 = TextureCoordinate() {vTiling=0.5}
  n12 = Constant2Vector() {r=0.7}
  n13 = TextureSample(UVs=Panner(Coordinate=Add(A=@n11, B=@n12)) {speedX=-2}) {texture=&T_Tile_RainbowColor}
  n14 = Desaturation(
   None=VectorParameter() {parameterName=Color, defaultValue=<rgba 1.5 1.4799 0.5675 1>}.RGB
   Fraction=ScalarParameter() {defaultValue=1, parameterName=DefaultColorDes}
  )
  n15 = TextureCoordinate()
  n16 = Multiply(A=TextureSample(UVs=Panner(Coordinate=@n15) {speedX=0.1}) {texture=&T_Inky_Smoke_Tile}.R) {constB=0.1}
  n17 = Add(A=@n16, B=TextureCoordinate() {uTiling=0.5, vTiling=0.5})
  n18 = Multiply(A=Time(), B=ScalarParameter() {defaultValue=-1.5, parameterName=Speed Direction}) {constB=-1.5}
  n19 = Power(Base=TextureSample(UVs=Panner(Coordinate=@n17, Time=@n18) {speedX=-0.8}) {samplerType=SAMPLERTYPE_LinearGrayscale, texture=/Game/Effects/Fort_Effects/Emotes/UltraLight/Texture/T_PlasmaCoils}.R) {constExponent=1.5}
  n20 = VectorParameter() {parameterName=ColorDark, defaultValue=<rgba 0.3931 0.401 0.1693 1>}
  n21 = Add(A=TextureCoordinate(), B=@n16)
  n22 = Power(
   Base=TextureSample(UVs=Panner(Coordinate=@n21, Time=@n18) {speedX=-1}) {samplerType=SAMPLERTYPE_Grayscale, texture=&T_Smoke_Trail_Soft}.R
   Exp=ScalarParameter() {defaultValue=4, parameterName=Exp}
  )
  n23 = Add(
   A=Desaturation(None=Power(Base=@n13.RGB), Fraction=ScalarParameter() {defaultValue=1, parameterName=ColorNoiseDes})
   B=Add(A=Multiply(A=@n14, B=@n19), B=Multiply(A=@n20.RGB, B=@n22))
  )
  MaterialFunctionCall(Emissive=Multiply(A=ParticleColor().RGB, B=Add(A=@n23))) {materialFunction=&MF_FX_EmissiveAdjustments}
  n24 = Add(A=@n16, B=TextureCoordinate())
  n25 = Multiply(
   A=TextureSample(UVs=Panner(Coordinate=@n24, Time=@n18) {speedX=-1.5}) {samplerType=SAMPLERTYPE_Grayscale, texture=&LowResBlurredNoise}.R
   B=@n22
  )
  n26 = ComponentMask(None=TextureCoordinate()) {r=true}
  n27 = Multiply(A=ParticleColor().A, B=Multiply(A=Add(A=@n22, B=@n25) {constB=0}, B=Sine(None=@n26) {period=2}))
  n28 = DepthFade(Opacity=@n27, FadeDistance=ScalarParameter() {defaultValue=10, parameterName=FadeDistance}) {fadeDistanceDefault=10}
  n29 = ShadingPathSwitch(Default=@n28, Deferred=@n28, Forward=@n27, Mobile=@n27)
  n30 = ScalarParameter() {parameterName="FadeSensitivity (optional) (S)"}
  n31 = Saturate(None=Multiply(A=MaterialFunctionCall("FadeSensitivity (optional)"=@n30) {materialFunction=&FadeBasedOnViewAngle}.Result) {constB=10})
  n32 = ComponentMask(None=TextureCoordinate()) {r=true}
  n33 = LinearInterpolate(Alpha=DynamicParameter() {paramNames [Param1,Param2,Param3,AppearMask]}.AppearMask) {constA=-1}
  n34 = Saturate(None=Multiply(A=Add(A=@n32, B=@n33) {constB=0}) {constB=3})
  n35 = ScalarParameter() {parameterName=UseAppearMask}
  n36 = ScalarParameter() {defaultValue=100, parameterName=WorldPosition_SphereMaskRadius}
  n37 = DynamicParameter() {parameterIndex=3, paramNames [WorldPosition_SphereMaskRadius control, Param2, Param3, Param4]}
  Multiply(
   A=Multiply(A=Multiply(A=@n29, B=@n31), B=LinearInterpolate(B=@n34, Alpha=@n35) {constA=1})
   B=SphereMask(A=WorldPosition().XYZ, B=ParticlePositionWS(), Radius=Add(A=@n36, B=@n37.WorldPosition_SphereMaskRadius control)) {attenuationRadius=65, hardnessPercent=50}
  )
  OneMinus(None=@n32)
  TextureCoordinate() {uTiling=3}
 ]
}
