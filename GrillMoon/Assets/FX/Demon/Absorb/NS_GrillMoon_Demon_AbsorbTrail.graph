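# Niagara effect for the demon's absorb: a dark purple-magenta trail arcing between a fixed point and the left or right hand, per User.LeftHand.
# Burst ribbons place particles by execution index along a Hermite spline rebuilt from the sampled socket.
# User Growth, Erosion and Flip drive the trail's reveal and retraction.
# Smoke and sparks follow the spline, ExitPortal meshes sit at the hand, and LeadParticle and AccentRibbons are disabled.
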NiagaraSystem /GrillMoon/Assets/FX/Demon/Absorb/NS_GrillMoon_Demon_AbsorbTrail
dependencies {
 usedDataInterfaces [
  &NiagaraDataInterfaceCurve
  &NiagaraDataInterfaceColorCurve
  &NiagaraDataInterfaceParticleRead
  &NiagaraDataInterfaceSpriteRendererInfo
  &NiagaraDataInterfaceSkeletalMesh
 ]
 usedDynamicInputs []
 usedModules [
  &self:SystemSpawnScript.SystemScriptSource.SystemScriptGraph.NiagaraNodeAssignment_0.SetVariables_#1
  &SystemState
 ]
}
emitters {
 AccentRibbons {
  bEnabled=false
  emitterUpdateScript [
   &EmitterState()
   &SpawnParticlesFromOtherEmitter(
    Attribute Reader=DataInterface() {propertyValues {EmitterBinding {bindingMode=Other, emitterName=LeadParticle}}}
    Spawn Rate=40
    Use Maximum Age=false
    Use Minimum Age=false
   )
  ]
  particleSpawnScript [
   &InitializeParticle(Lifetime=0.7, UsePositionOffset=false, Write Material Random=true)
   &SampleParticlesFromOtherEmitter(Particle ID Sampling=::NewEnumerator1, Sprite Size Sampling:ENiagara_DirectReadSamplingMode=::NewEnumerator0)
   /SetVariables(Particles.RibbonWidth=&Lerp_Float(
     A=0
     B=400
     Alpha=&FloatFromCurve(
      FloatCurve=DataInterface() {propertyValues {Curve {keys [
          {arriveTangent=0.8557, interpMode=RCIM_Cubic, leaveTangent=0.8557, tangentMode=RCTM_User, time=0, value=0}
          {arriveTangent=2, leaveTangent=2, time=0.25, value=1}
          {interpMode=RCIM_Cubic, tangentMode=RCTM_User, time=0.5, value=1}
          <at 1 0 RCIM_Cubic>
         ]}}}
      CurveIndex=&MakeFloatFromVector2d(Vector 2D=$Output.SampleParticlesFromOtherEmitter.SampledSpriteSize)
     )
    )) {isSetParameters=true}
  ]
  particleUpdateScript [
   &SolveForcesAndVelocity(Clamp Velocity=false, Limit Acceleration=false)
   &ParticleState()
   &ScaleColor(
    Scale Alpha=&FloatFromCurve(
     FloatCurve=DataInterface() {propertyValues {Curve {keys [
         {arriveTangent=6.4598, interpMode=RCIM_Cubic, leaveTangent=6.4598, tangentMode=RCTM_User, time=0, value=0}
         {interpMode=RCIM_Cubic, tangentMode=RCTM_User, time=0.5, value=1}
         {tangentMode=RCTM_User, time=1, value=0}
        ]}}}
     Scale Curve=$User.SpawnRate
    )
    Scale RGB=&MakeVectorFromLinearColorRGB(Color=&ColorFromCurve(ColorCurve=DataInterface() {propertyValues {
        AlphaCurve {keys [<at 0 1 RCIM_Cubic>]}
        BlueCurve {keys [<at 0.2504 0 RCIM_Cubic>,<at 0.5425 0.4349>,<at 0.9443 0.7865 RCIM_Cubic>]}
        GreenCurve {keys [<at 0.2504 0.506 RCIM_Cubic>,<at 0.5425 0>,<at 0.9443 0 RCIM_Cubic>]}
        RedCurve {keys [<at 0.2504 5 RCIM_Cubic>,<at 0.5425 1>,<at 0.9443 0.3314 RCIM_Cubic>]}
       }}))
    ScaleA=true
    ScaleRGB=true
   )
   &DynamicMaterialParameters(
    Index 0 Param 1=&RandomRangeFloat(Minimum=0, Maximum=1)
    Module.Index0Param0Write=true
    Module.Index0Param1Write=true
    Module.Index0Param2Write=true
    Module.Index0Param3Write=true
   )
   &ScaleRibbonWidth(Ribbon Width Scale=&FloatFromCurve(
     FloatCurve=DataInterface() {propertyValues {Curve {keys [
         {interpMode=RCIM_Cubic, tangentMode=RCTM_User, time=0, value=0}
         {interpMode=RCIM_Cubic, tangentMode=RCTM_User, time=0.5, value=1}
         {tangentMode=RCTM_User, time=1, value=0}
        ]}}}
     Scale Curve=&RandomRangeFloat(Minimum=0.5, Maximum=1)
    ))
  ]
  rendererClasses [&NiagaraRibbonRendererProperties]
  renderers [{rendererClass=&NiagaraRibbonRendererProperties}]
 }
 ExitPortal {
  emitterUpdateScript [
   &EmitterState()
   &SpawnBurst_Instantaneous(Spawn Group=1, Use Spawn Probability=false, UseLoopCountLimit=true)
  ]
  particleSpawnScript [
   &InitializeParticle(
    Color Mode=::NewEnumerator0
    Lifetime=&MaskFloatBySpawnGroup(Masked Float=0.9, Passthrough Float=0.7)
    UsePositionOffset=false
    Write Material Random=true
   )
   &CameraOffset(Camera Offset Amount=500)
  ]
  particleUpdateScript [
   &ParticleState(Kill Particles When Lifetime Has Elapsed=false, Let Infinitely Lived Particles Die When Emitter Deactivates=true)
   &SkeletalMeshLocation(
    Bone / Socket Index=&MakeCustomIntFromBool(Bool=$User.LeftHand, True INT=1, False INT=0)
    Bone Sampling Mode=::NewEnumerator4
    MaskSpawnGroup=false
    Orientation Sampling=::NewEnumerator1
    Sampled Position Offset=&MakeCustomVectorFromBool(Bool=$User.LeftHand, True Vector=<-800 350 0>, False Vector=<1000 -350 0>)
    Skeletal Mesh=$System.Skeletal Mesh
   )
   /SetVariables(
    Particles.Color2=&ColorFromCurve(
     ColorCurve=DataInterface() {propertyValues {
       AlphaCurve {keys [<at 0 0 RCIM_Cubic>,<at 0.0826 3 RCIM_Cubic>]}
       BlueCurve {keys [<at 0.4494 0 RCIM_Cubic>]}
       GreenCurve {keys [<at 0.4494 0 RCIM_Cubic>]}
       RedCurve {keys [<at 0.4494 0 RCIM_Cubic>]}
      }}
     CurveIndex=$Emitter.Age
    )
    Particles.Color_Outline:LinearColor=<rgba 0.9894 0 1.5 1>
   ) {isSetParameters=true}
   /SetVariables(
    Particles.DynamicParam3:Vector4f=&MakeVector4(
     X=&FloatFromCurve(
      FloatCurve=DataInterface() {propertyValues {Curve {keys [
          {arriveTangent=5.6729, interpMode=RCIM_Cubic, leaveTangent=5.6729, tangentMode=RCTM_User, time=0.4, value=0}
          {arriveTangent=-0.2155, interpMode=RCIM_Cubic, leaveTangent=-0.2155, tangentMode=RCTM_User, time=0.5, value=1.2}
          {arriveTangent=-0.6116, interpMode=RCIM_Cubic, leaveTangent=-0.6116, tangentMode=RCTM_User, time=1, value=1}
         ]}}}
      CurveIndex=$Emitter.Age
      Scale Curve=&Lerp_Float(A=14, B=0, Alpha=$User.Erosion)
     )
     Y=&FloatFromCurve(
      FloatCurve=DataInterface() {propertyValues {Curve {keys [
          {arriveTangent=12.6541, interpMode=RCIM_Cubic, leaveTangent=12.6541, tangentMode=RCTM_User, time=0, value=0}
          {arriveTangent=1.2635, interpMode=RCIM_Cubic, leaveTangent=1.2635, tangentMode=RCTM_User, time=0.3, value=1.625}
         ]}}}
      CurveIndex=$Emitter.Age
      Scale Curve=25
     )
     Z=0
     W=1
    )
    Particles.Scale2:Vector3f=&Multiply_VectorByFloat(
     Vector=&Multiply_Vector(
      A=&VectorFromFloat(Value=&FloatFromCurve(
        FloatCurve=DataInterface() {propertyValues {Curve {keys [
            {arriveTangent=0.0748, interpMode=RCIM_Cubic, leaveTangent=0.0748, tangentMode=RCTM_User, time=0.3, value=0}
            {arriveTangent=4.6251, interpMode=RCIM_Cubic, leaveTangent=7.8087, tangentMode=RCTM_Break, time=0.8, value=1.4}
           ]}}}
        CurveIndex=$Emitter.Age
       ))
      B=<0.72 0.72 0.32>
     )
     Float=&Lerp_Float(A=10, B=0, Alpha=$User.Erosion)
    )
    Particles.Scale3:Vector3f=&Multiply_VectorByFloat(
     Vector=&Multiply_Vector(
      A=&VectorFromFloat(Value=&FloatFromCurve(
        FloatCurve=DataInterface() {propertyValues {Curve {keys [
            {arriveTangent=0.0748, interpMode=RCIM_Cubic, leaveTangent=0.0748, tangentMode=RCTM_User, time=0.3, value=0}
            {arriveTangent=4.6251, interpMode=RCIM_Cubic, leaveTangent=7.8087, tangentMode=RCTM_Break, time=0.8, value=1.4}
           ]}}}
        CurveIndex=$Emitter.Age
       ))
      B=<0.8 0.8 0.36>
     )
     Float=&Lerp_Float(A=12, B=0, Alpha=$User.Erosion)
    )
   ) {isSetParameters=true}
  ]
  rendererClasses [&NiagaraMeshRendererProperties]
  renderers [
   {rendererClass=&NiagaraMeshRendererProperties, rendererIndex=0}
   {rendererClass=&NiagaraMeshRendererProperties, rendererIndex=1}
  ]
 }
 LeadParticle {
  bEnabled=false
  emitterUpdateScript [
   &EmitterState()
   &SpawnRate(
    SpawnRate=&FloatFromCurve(
     FloatCurve=DataInterface() {propertyValues {Curve {keys [{tangentMode=RCTM_User, time=0.25, value=0}, {tangentMode=RCTM_User, time=0.7, value=1}]}}}
     CurveIndex=$User.SpawnRate
     Scale Curve=2
    )
    Use Spawn Probability=false
   )
  ]
  particleSpawnScript [
   &InitializeParticle(
    Lifetime Max=3
    Lifetime Min=1.5
    Lifetime Mode=::NewEnumerator1
    Sprite Size Mode=::NewEnumerator3
    Uniform Sprite Size=1
    UsePositionOffset=false
    Write Material Random=true
   )
   &SkeletalMeshLocation(
    Bone / Socket Index=&MakeCustomIntFromBool(Bool=$User.LeftHand, True INT=1, False INT=0)
    Bone Sampling Mode=::NewEnumerator4
    MaskSpawnGroup=false
    Orientation Sampling=::NewEnumerator1
    Skeletal Mesh=$System.Skeletal Mesh
    Write To Particle Position=false
   )
  ]
  particleUpdateScript [
   &ParticleState()
   /Game/Effects/Niagara/Modules/Curves/FN_HermiteSpline(
    Compute Velocity=true
    Direction End:Vector3f=&MakeCustomVectorFromBool(
     Bool=$User.LeftHand
     True Vector=&MakeCustomVectorFromBool(
      Bool=$User.LeftHand
      True Vector=&MakeVector(
       X=0
       Y=0
       Z=&Multiply_Float(
        A=-9000
        B=&FloatFromCurve(FloatCurve=DataInterface() {propertyValues {Curve {keys [
             {interpMode=RCIM_Cubic, tangentMode=RCTM_User, time=0.5, value=1}
             {arriveTangent=-5.1701, interpMode=RCIM_Cubic, leaveTangent=-5.1701, tangentMode=RCTM_User, time=1, value=0}
            ]}}})
       )
      )
      False Vector=<0 0 0>
     )
     False Vector=<0 0 -4500>
    )
    Direction Start:Vector3f=&MakeCustomVectorFromBool(
     Bool=$User.LeftHand
     True Vector=&MakeVector(
      X=&Multiply_Float(
       A=40000
       B=&FloatFromCurve(FloatCurve=DataInterface() {propertyValues {Curve {keys [{interpMode=RCIM_Cubic, tangentMode=RCTM_User, time=0.2, value=0}, {interpMode=RCIM_Cubic, tangentMode=RCTM_User, time=1, value=1}]}}})
      )
      Y=0
      Z=29000
     )
     False Vector=<0 0 43750>
    )
    Position Blend:NiagaraFloat=$Particles.NormalizedAge
    Position End:Vector3f=&Add_Vector(
     A=$Particles.SkeletalMeshLocation.SampledPosition
     B=&RotateVectorByQuaternion(
      Vector To Rotate=&MakeCustomVectorFromBool(Bool=$User.LeftHand, True Vector=<-900 350 0>, False Vector=<1000 -350 0>)
      Quaternion=$Particles.SkeletalMeshLocation.SampledRotation
     )
    )
    Position Start:Vector3f=<-57600 42800 11500>
   )
   &ShapeLocation(
    Apply To Particle Position=::NewEnumerator1
    MaskSpawnGroup=false
    Override Seed=false
    Ring / Disc Distribution Mode:ENiagaraSphereDistributionMode=::NewEnumerator1
    Ring Radius=&RandomRangeFloat(Minimum=600, Maximum=1250)
    Rotation Axis=$Output.FN_HermiteSpline.Velocity
    Rotation Mode=::NewEnumerator1
    Shape Primitive=::NewEnumerator4
    U Position=&Frac_Float(Float=&Multiply_Float(
      A=$Particles.NormalizedAge
      B=&Multiply_Float(
       A=&MakeFloatFromInt(Int=&RandomRangeInt(Minimum=1, Maximum=4))
       B=&UniformAOrBFloat(A=-1, Evaluation Type=::NewEnumerator0, Recalculate Random Each Loop=true, Override Seed=false)
      )
     ))
    UseOwnerScale=true
   )
   /SetVariables(Particles.Position=&Add_Vector(A=$Particles.ShapeLocation.ShapeVector, B=$Output.FN_HermiteSpline.Position)) {isSetParameters=true}
   &ScaleSpriteSize()
  ]
  rendererClasses [&NiagaraSpriteRendererProperties]
  renderers [{rendererClass=&NiagaraSpriteRendererProperties}]
 }
 MainTrail {
  emitterUpdateScript [
   &EmitterState()
   &SpawnBurst_Instantaneous(Spawn Count=40, Use Spawn Probability=false, UseLoopCountLimit=true)
  ]
  particleSpawnScript [&InitializeParticle(Ribbon Width=2150, Ribbon Width Mode=::NewEnumerator1, UsePositionOffset=false, Write Material Random=true)]
  particleUpdateScript [
   &ParticleState(
    Kill Particles When Lifetime Has Elapsed=false
    Let Infinitely Lived Particles Die When Emitter Deactivates=true
    Loop Particles Lifetime=true
   )
   &SkeletalMeshLocation(
    Bone / Socket Index=&MakeCustomIntFromBool(Bool=$User.LeftHand, True INT=1, False INT=0)
    Bone Sampling Mode=::NewEnumerator4
    MaskSpawnGroup=false
    Orientation Sampling=::NewEnumerator1
    Skeletal Mesh=$System.Skeletal Mesh
    Write To Particle Position=false
   )
   &FN_HermiteSpline(
    Direction End:Vector3f=&MakeCustomVectorFromBool(
     Bool=$User.LeftHand
     True Vector=&MakeVector(
      X=0
      Y=0
      Z=&Multiply_Float(
       A=16000
       B=&FloatFromCurve(
        FloatCurve=DataInterface() {propertyValues {Curve {keys [
            {interpMode=RCIM_Cubic, tangentMode=RCTM_User, time=0.5, value=1}
            {arriveTangent=-5.1701, interpMode=RCIM_Cubic, leaveTangent=-5.1701, tangentMode=RCTM_User, time=1, value=0}
           ]}}}
        CurveIndex=&ReturnNormalizedExecIndex()
       )
      )
     )
     False Vector=<0 0 -4500>
    )
    Direction Start:Vector3f=&MakeCustomVectorFromBool(
     Bool=$User.LeftHand
     True Vector=&MakeVector(
      X=&Multiply_Float(
       A=80000
       B=&FloatFromCurve(
        FloatCurve=DataInterface() {propertyValues {Curve {keys [{interpMode=RCIM_Cubic, tangentMode=RCTM_User, time=0.2, value=0}, {interpMode=RCIM_Cubic, tangentMode=RCTM_User, time=1, value=1}]}}}
        CurveIndex=&ReturnNormalizedExecIndex()
       )
      )
      Y=16000
      Z=8000
     )
     False Vector=<0 0 43750>
    )
    Position Blend:NiagaraFloat=&ReturnNormalizedExecIndex()
    Position End:Vector3f=&Add_Vector(
     A=$Particles.SkeletalMeshLocation.SampledPosition
     B=&RotateVectorByQuaternion(
      Vector To Rotate=&MakeCustomVectorFromBool(Bool=$User.LeftHand, True Vector=<-900 350 0>, False Vector=<1000 -350 0>)
      Quaternion=$Particles.SkeletalMeshLocation.SampledRotation
     )
    )
    Position Start:Vector3f=<-57600 42800 11500>
    Write Particles Position=true
   )
   &ScaleRibbonWidth(Ribbon Width Scale=&FloatFromCurve(
     FloatCurve=DataInterface() {propertyValues {Curve {keys [
         {arriveTangent=5.7924, interpMode=RCIM_Cubic, leaveTangent=5.7924, tangentMode=RCTM_User, time=0, value=0.4}
         {tangentMode=RCTM_User, time=0.3, value=1}
         {arriveTangent=-0.0405, interpMode=RCIM_Cubic, leaveTangent=-0.0405, tangentMode=RCTM_User, time=0.7, value=1}
         {arriveTangent=-4.06, interpMode=RCIM_Cubic, leaveTangent=-4.06, tangentMode=RCTM_User, time=1, value=0.5}
        ]}}}
     CurveIndex=&ReturnNormalizedExecIndex()
     Scale Curve=&FloatFromCurve(
      FloatCurve=DataInterface() {propertyValues {Curve {keys [
          {interpMode=RCIM_Cubic, tangentMode=RCTM_User, time=0, value=0.7}
          {interpMode=RCIM_Cubic, tangentMode=RCTM_User, time=0.5, value=1.25}
          {tangentMode=RCTM_User, time=1, value=0.7}
         ]}}}
      CurveIndex=&Frac_Float(Float=&Add_Float(
        A=&Multiply_Float(A=$Particles.NormalizedAge, B=2)
        B=&Multiply_Float(A=&Lerp_Float(A=1, B=0, Alpha=&ReturnNormalizedExecIndex()), B=2)
       ))
      Scale Curve=&Lerp_Float(
       A=&FloatFromCurve(
        FloatCurve=DataInterface() {propertyValues {Curve {keys [{tangentMode=RCTM_User, time=0, value=0}, {tangentMode=RCTM_User, time=1, value=1}]}}}
        CurveIndex=&ClampFloat(Float=&Subtract_Float(
          A=&RemapRange(Input Value To Remap=$User.Growth, Input Min=0, Input Max=1, Output Min=0, Output Max=2)
          B=&ReturnNormalizedExecIndex()
         ))
       )
       B=&FloatFromCurve(
        FloatCurve=DataInterface() {propertyValues {Curve {keys [{tangentMode=RCTM_User, time=0, value=1}, {tangentMode=RCTM_User, time=1, value=0}]}}}
        CurveIndex=&ClampFloat(Float=&Subtract_Float(
          A=&RemapRange(Input Value To Remap=$User.Erosion, Input Min=0, Input Max=1, Output Min=0, Output Max=2)
          B=&ReturnNormalizedExecIndex()
         ))
       )
       Alpha=$User.Flip
      )
     )
    ))
   &DynamicMaterialParameters(
    Index 0 Param 1=0.75
    Index 0 Param 3=1.5
    Index 0 Param 4=2
    Index 1 Param 1=50
    Index 2 Param 1=0.35
    Index 2 Param 3=3.2
    Index 2 Param 4=2
    Index 3 Param 1=0.525
    Index 3 Param 3=2.5
    Index 3 Param 4=3
    Module.Index0Param0Write=true
    Module.Index0Param1Write=true
    Module.Index0Param2Write=true
    Module.Index0Param3Write=true
    Module.Index1Param0Write=true
    Module.Index1Param1Write=true
    Module.Index1Param2Write=true
    Module.Index1Param3Write=true
    Module.Index2Param0Write=true
    Module.Index2Param1Write=true
    Module.Index2Param2Write=true
    Module.Index2Param3Write=true
    Module.Index3Param0Write=true
    Module.Index3Param1Write=true
    Module.Index3Param2Write=true
    Module.Index3Param3Write=true
    Write Parameter Index 1=true
    Write Parameter Index 2=true
    Write Parameter Index 3=true
   )
   &ScaleColor(
    Scale Alpha=&FloatFromCurve(
     FloatCurve=DataInterface() {propertyValues {Curve {keys [
         {tangentMode=RCTM_User, time=0, value=1}
         {arriveTangent=-0.0405, interpMode=RCIM_Cubic, leaveTangent=-0.0405, tangentMode=RCTM_User, time=0.7009, value=1}
         {arriveTangent=-8.0089, interpMode=RCIM_Cubic, leaveTangent=-8.0089, tangentMode=RCTM_User, time=0.95, value=0}
        ]}}}
     CurveIndex=&ReturnNormalizedExecIndex()
    )
    Scale RGB=&MakeVectorFromLinearColorRGB(Color=&MultiplyLinearColor(
      Linear Color=&ColorFromCurve(
       ColorCurve=DataInterface() {propertyValues {
         AlphaCurve {keys [<at 0 1 RCIM_Cubic>]}
         BlueCurve {keys [<at 0.9 0.7448 RCIM_Cubic>]}
         GreenCurve {keys [<at 0.9 0.0124 RCIM_Cubic>]}
         RedCurve {keys [<at 0.9 0.1862 RCIM_Cubic>]}
        }}
       CurveIndex=&ReturnNormalizedExecIndex()
      )
      Scale Factor=&MakeVector4_FromLinearColor(LinearColor=&ColorFromCurve(
        ColorCurve=DataInterface() {propertyValues {
          AlphaCurve {keys [<at 0 1 RCIM_Cubic>,<at 1 1>]}
          BlueCurve {keys [<at 0 0.1268>, <at 0.5 0.0194 RCIM_Cubic>, {interpMode=RCIM_Cubic, leaveTangent=0.2737, time=1, value=0.1268}]}
          GreenCurve {keys [<at 0 0>, <at 0.5 0.0131 RCIM_Cubic>, {interpMode=RCIM_Cubic, leaveTangent=-0.0334, time=1, value=0}]}
          RedCurve {keys [<at 0 1.2>, <at 0.5 0.07 RCIM_Cubic>, {interpMode=RCIM_Cubic, leaveTangent=2.879, time=1, value=1.2}]}
         }}
        CurveIndex=&Frac_Float(Float=&Add_Float(
          A=&Multiply_Float(A=$Particles.NormalizedAge, B=2)
          B=&Multiply_Float(A=&Lerp_Float(A=1, B=0, Alpha=&ReturnNormalizedExecIndex()), B=4)
         ))
       ))
     ))
    ScaleA=true
    ScaleRGB=true
   )
   /SetVariables(
    Particles.InnerColor:LinearColor=&MultiplyLinearColor(
     Linear Color=&ColorFromCurve(
      ColorCurve=DataInterface() {propertyValues {
        AlphaCurve {keys [<at 0 1 RCIM_Cubic>]}
        BlueCurve {keys [<at 0.975 0.02 RCIM_Cubic>]}
        GreenCurve {keys [<at 0.975 0 RCIM_Cubic>]}
        RedCurve {keys [<at 0.975 0.012 RCIM_Cubic>]}
       }}
      CurveIndex=&ReturnNormalizedExecIndex()
     )
     Scale Factor=&MakeFloat4FromVectorAndFloat(
      "Vector (RGB)"=&VectorFromFloat(Value=&Lerp_Float(
        A=1
        B=1
        Alpha=&FloatFromCurve(
         FloatCurve=DataInterface() {propertyValues {Curve {keys [<at 0.5 0 RCIM_Cubic>, {arriveTangent=5.4291, tangentMode=RCTM_User, time=1, value=1}]}}}
         CurveIndex=&ReturnNormalizedExecIndex()
        )
       ))
      "Float (Alpha)"=&FloatFromCurve(
       FloatCurve=DataInterface() {propertyValues {Curve {keys [
           {tangentMode=RCTM_User, time=0, value=1}
           {arriveTangent=-0.0405, interpMode=RCIM_Cubic, leaveTangent=-0.0405, tangentMode=RCTM_User, time=0.7009, value=1}
           {arriveTangent=-8.0089, interpMode=RCIM_Cubic, leaveTangent=-8.0089, tangentMode=RCTM_User, time=0.95, value=0}
          ]}}}
       CurveIndex=&ReturnNormalizedExecIndex()
      )
     )
    )
    Particles.InnerRibbonSize:NiagaraFloat=&Multiply_Float(
     A=2150
     B=&FloatFromCurve(
      FloatCurve=DataInterface() {propertyValues {Curve {keys [
          {arriveTangent=6.1118, interpMode=RCIM_Cubic, leaveTangent=6.1118, tangentMode=RCTM_User, time=0, value=0.4}
          {tangentMode=RCTM_User, time=0.3, value=1}
          {arriveTangent=-0.0405, interpMode=RCIM_Cubic, leaveTangent=-0.0405, tangentMode=RCTM_User, time=0.7, value=1}
          {arriveTangent=-4.06, interpMode=RCIM_Cubic, leaveTangent=-4.06, tangentMode=RCTM_User, time=1, value=0.5}
         ]}}}
      CurveIndex=&ReturnNormalizedExecIndex()
      Scale Curve=&FloatFromCurve(
       FloatCurve=DataInterface() {propertyValues {Curve {keys [
           {interpMode=RCIM_Cubic, tangentMode=RCTM_User, time=0, value=0.7}
           {interpMode=RCIM_Cubic, tangentMode=RCTM_User, time=0.5, value=1.25}
           {tangentMode=RCTM_User, time=1, value=0.7}
          ]}}}
       CurveIndex=&Frac_Float(Float=&Add_Float(
         A=&Multiply_Float(A=$Particles.NormalizedAge, B=2)
         B=&Multiply_Float(A=&Lerp_Float(A=1, B=0, Alpha=&ReturnNormalizedExecIndex()), B=2)
        ))
       Scale Curve=&Lerp_Float(
        A=&FloatFromCurve(
         FloatCurve=DataInterface() {propertyValues {Curve {keys [{tangentMode=RCTM_User, time=0, value=0}, {tangentMode=RCTM_User, time=1, value=1}]}}}
         CurveIndex=&ClampFloat(Float=&Subtract_Float(
           A=&RemapRange(Input Value To Remap=$User.Growth, Input Min=0, Input Max=1, Output Min=0, Output Max=2)
           B=&ReturnNormalizedExecIndex()
          ))
        )
        B=&FloatFromCurve(
         FloatCurve=DataInterface() {propertyValues {Curve {keys [{tangentMode=RCTM_User, time=0, value=1}, {tangentMode=RCTM_User, time=1, value=0}]}}}
         CurveIndex=&ClampFloat(Float=&Subtract_Float(
           A=&RemapRange(Input Value To Remap=$User.Erosion, Input Min=0, Input Max=1, Output Min=0, Output Max=2)
           B=&ReturnNormalizedExecIndex()
          ))
        )
        Alpha=$User.Flip
       )
      )
     )
    )
    Particles.SecondaryColor:LinearColor=&MultiplyLinearColor(
     Linear Color=&ColorFromCurve(
      ColorCurve=DataInterface() {propertyValues {
        AlphaCurve {keys [<at 0 1 RCIM_Cubic>]}
        BlueCurve {keys [<at 0.97 0.0771 RCIM_Cubic>]}
        GreenCurve {keys [<at 0.97 0 RCIM_Cubic>]}
        RedCurve {keys [<at 0.97 0.0781 RCIM_Cubic>]}
       }}
      CurveIndex=&ReturnNormalizedExecIndex()
     )
     Scale Factor=&MakeFloat4FromVectorAndFloat(
      "Vector (RGB)"=&VectorFromFloat(Value=&Lerp_Float(
        A=1
        B=1
        Alpha=&FloatFromCurve(
         FloatCurve=DataInterface() {propertyValues {Curve {keys [<at 0.5 0 RCIM_Cubic>, {arriveTangent=5.4291, tangentMode=RCTM_User, time=1, value=1}]}}}
         CurveIndex=&ReturnNormalizedExecIndex()
        )
       ))
      "Float (Alpha)"=&FloatFromCurve(
       FloatCurve=DataInterface() {propertyValues {Curve {keys [
           {tangentMode=RCTM_User, time=0, value=1}
           {arriveTangent=-0.0405, interpMode=RCIM_Cubic, leaveTangent=-0.0405, tangentMode=RCTM_User, time=0.7009, value=1}
           {arriveTangent=-8.0089, interpMode=RCIM_Cubic, leaveTangent=-8.0089, tangentMode=RCTM_User, time=0.95, value=0}
          ]}}}
       CurveIndex=&ReturnNormalizedExecIndex()
      )
     )
    )
    Particles.SecondaryRibbonSize:NiagaraFloat=&Multiply_Float(
     A=2250
     B=&FloatFromCurve(
      FloatCurve=DataInterface() {propertyValues {Curve {keys [
          {arriveTangent=5.6931, interpMode=RCIM_Cubic, leaveTangent=5.6931, tangentMode=RCTM_User, time=0, value=0.4}
          {tangentMode=RCTM_User, time=0.3, value=1}
          {arriveTangent=-0.0405, interpMode=RCIM_Cubic, leaveTangent=-0.0405, tangentMode=RCTM_User, time=0.7, value=1}
          {arriveTangent=-4.06, interpMode=RCIM_Cubic, leaveTangent=-4.06, tangentMode=RCTM_User, time=1, value=0.5}
         ]}}}
      CurveIndex=&ReturnNormalizedExecIndex()
      Scale Curve=&FloatFromCurve(
       FloatCurve=DataInterface() {propertyValues {Curve {keys [
           {interpMode=RCIM_Cubic, tangentMode=RCTM_User, time=0, value=0.7}
           {interpMode=RCIM_Cubic, tangentMode=RCTM_User, time=0.5, value=1.25}
           {tangentMode=RCTM_User, time=1, value=0.7}
          ]}}}
       CurveIndex=&Frac_Float(Float=&Add_Float(
         A=&Multiply_Float(A=$Particles.NormalizedAge, B=2)
         B=&Multiply_Float(A=&Lerp_Float(A=1, B=0, Alpha=&ReturnNormalizedExecIndex()), B=2)
        ))
       Scale Curve=&Lerp_Float(
        A=&FloatFromCurve(
         FloatCurve=DataInterface() {propertyValues {Curve {keys [{tangentMode=RCTM_User, time=0, value=0}, {tangentMode=RCTM_User, time=1, value=1}]}}}
         CurveIndex=&ClampFloat(Float=&Subtract_Float(
           A=&RemapRange(Input Value To Remap=$User.Growth, Input Min=0, Input Max=1, Output Min=0, Output Max=2)
           B=&ReturnNormalizedExecIndex()
          ))
        )
        B=&FloatFromCurve(
         FloatCurve=DataInterface() {propertyValues {Curve {keys [{tangentMode=RCTM_User, time=0, value=1}, {tangentMode=RCTM_User, time=1, value=0}]}}}
         CurveIndex=&ClampFloat(Float=&Subtract_Float(
           A=&RemapRange(Input Value To Remap=$User.Erosion, Input Min=0, Input Max=1, Output Min=0, Output Max=2)
           B=&ReturnNormalizedExecIndex()
          ))
        )
        Alpha=$User.Flip
       )
      )
     )
    )
    Particles.SoftAirColor:LinearColor=&MultiplyLinearColor(
     Linear Color=<rgba 0.125 0 0.058 0.25>
     Scale Factor=&MakeFloat4FromVectorAndFloat(
      "Vector (RGB)"=<1 1 1>
      "Float (Alpha)"=&FloatFromCurve(
       FloatCurve=DataInterface() {propertyValues {Curve {keys [{tangentMode=RCTM_User, time=0, value=0}, {tangentMode=RCTM_User, time=1, value=1}]}}}
       CurveIndex=&FloatFromCurve(
        FloatCurve=DataInterface() {propertyValues {Curve {keys [{tangentMode=RCTM_User, time=0, value=0}, {tangentMode=RCTM_User, time=1, value=1}]}}}
        CurveIndex=&ClampFloat(Float=&Subtract_Float(
          A=&RemapRange(Input Value To Remap=$User.Growth, Input Min=0, Input Max=1, Output Min=0, Output Max=2)
          B=&ReturnNormalizedExecIndex()
         ))
       )
      )
     )
    )
    Particles.SoftAirDynm:Vector4f=<0.5 0 0 0>
    Particles.SoftAirSpriteSize:Vector2f=&Vector2DFromFloat(Value=&FloatFromCurve(
      FloatCurve=DataInterface() {propertyValues {Curve {keys [
          {arriveTangent=5.2686, interpMode=RCIM_Cubic, leaveTangent=5.2686, tangentMode=RCTM_User, time=0, value=0.3}
          {tangentMode=RCTM_User, time=0.3, value=1}
          {arriveTangent=-0.0405, interpMode=RCIM_Cubic, leaveTangent=-0.0405, tangentMode=RCTM_User, time=0.7, value=1}
          {arriveTangent=-4.06, interpMode=RCIM_Cubic, leaveTangent=-4.06, tangentMode=RCTM_User, time=1, value=0.5}
         ]}}}
      CurveIndex=&ReturnNormalizedExecIndex()
      Scale Curve=&Lerp_Float(
       A=&FloatFromCurve(
        FloatCurve=DataInterface() {propertyValues {Curve {keys [{tangentMode=RCTM_User, time=0, value=0}, {tangentMode=RCTM_User, time=1, value=1}]}}}
        CurveIndex=&ClampFloat(Float=&Subtract_Float(
          A=&RemapRange(Input Value To Remap=$User.Growth, Input Min=0, Input Max=1, Output Min=0, Output Max=2)
          B=&ReturnNormalizedExecIndex()
         ))
        Scale Curve=2750
       )
       B=&FloatFromCurve(
        FloatCurve=DataInterface() {propertyValues {Curve {keys [{tangentMode=RCTM_User, time=0, value=1}, {tangentMode=RCTM_User, time=1, value=0}]}}}
        CurveIndex=&ClampFloat(Float=&Subtract_Float(
          A=&RemapRange(Input Value To Remap=$User.Erosion, Input Min=0, Input Max=1, Output Min=0, Output Max=2)
          B=&ReturnNormalizedExecIndex()
         ))
        Scale Curve=2750
       )
       Alpha=$User.Flip
      )
     ))
   ) {isSetParameters=true}
  ]
  rendererClasses [&NiagaraRibbonRendererProperties, &NiagaraSpriteRendererProperties]
  renderers [
   {rendererClass=&NiagaraRibbonRendererProperties}
   {rendererClass=&NiagaraRibbonRendererProperties, rendererIndex=1}
   {rendererClass=&NiagaraRibbonRendererProperties, rendererIndex=2}
   {rendererClass=&NiagaraSpriteRendererProperties, rendererIndex=3}
  ]
 }
 Smoke {
  emitterUpdateScript [
   &EmitterState()
   &SpawnRate(SpawnRate=25, Use Spawn Probability=false)
  ]
  particleSpawnScript [
   &InitializeParticle(
    Color=<rgba 0.0635 0 0.0677 1>
    Lifetime Min=0.7
    Lifetime Mode=::NewEnumerator1
    Sprite Rotation Mode=::NewEnumerator2
    Sprite Size Mode=::NewEnumerator1
    Uniform Sprite Size Max=2750
    Uniform Sprite Size Min=700
    UsePositionOffset=true
    Write Material Random=true
   )
   &CameraOffset(Camera Offset Amount=-20)
   &SkeletalMeshLocation(
    Bone / Socket Index=&MakeCustomIntFromBool(Bool=$User.LeftHand, True INT=1, False INT=0)
    Bone Sampling Mode=::NewEnumerator4
    MaskSpawnGroup=false
    Orientation Sampling=::NewEnumerator1
    Skeletal Mesh=$System.Skeletal Mesh
    Write To Particle Position=false
   )
   /SetVariables(Particles.LocationOnSpline:NiagaraFloat=&RemapRange(
     Input Value To Remap=&RandomRangeFloat(Minimum=0.1, Maximum=0.9)
     Input Min=0
     Input Max=1
     Output Min=0
     Output Max=&ClampFloat(Float=&Subtract_Float(
       A=&RemapRange(Input Value To Remap=$User.Growth, Input Min=0, Input Max=1, Output Min=0, Output Max=2)
       B=&ReturnNormalizedExecIndex()
      ))
    )) {isSetParameters=true, module=SetVariables_#2}
   &FN_HermiteSpline(
    Direction End:Vector3f=&MakeCustomVectorFromBool(
     Bool=$User.LeftHand
     True Vector=&MakeCustomVectorFromBool(
      Bool=$User.LeftHand
      True Vector=&MakeVector(
       X=0
       Y=0
       Z=&Multiply_Float(
        A=16000
        B=&FloatFromCurve(
         FloatCurve=DataInterface() {propertyValues {Curve {keys [
             {interpMode=RCIM_Cubic, tangentMode=RCTM_User, time=0.5, value=1}
             {arriveTangent=-5.1701, interpMode=RCIM_Cubic, leaveTangent=-5.1701, tangentMode=RCTM_User, time=1, value=0}
            ]}}}
         CurveIndex=$Particles.LocationOnSpline
        )
       )
      )
      False Vector=<0 0 0>
     )
     False Vector=<0 0 -4500>
    )
    Direction Start:Vector3f=&MakeCustomVectorFromBool(
     Bool=$User.LeftHand
     True Vector=&MakeVector(
      X=&Multiply_Float(
       A=80000
       B=&FloatFromCurve(
        FloatCurve=DataInterface() {propertyValues {Curve {keys [{interpMode=RCIM_Cubic, tangentMode=RCTM_User, time=0.2, value=0}, {interpMode=RCIM_Cubic, tangentMode=RCTM_User, time=1, value=1}]}}}
        CurveIndex=$Particles.LocationOnSpline
       )
      )
      Y=16000
      Z=8000
     )
     False Vector=<0 0 43750>
    )
    Position Blend:NiagaraFloat=$Particles.LocationOnSpline
    Position End:Vector3f=&Add_Vector(
     A=$Particles.SkeletalMeshLocation.SampledPosition
     B=&RotateVectorByQuaternion(
      Vector To Rotate=&MakeCustomVectorFromBool(Bool=$User.LeftHand, True Vector=<-900 350 0>, False Vector=<1000 -350 0>)
      Quaternion=$Particles.SkeletalMeshLocation.SampledRotation
     )
    )
    Position Start:Vector3f=<-57600 42800 11500>
   )
   &ShapeLocation(MaskSpawnGroup=false, Override Seed=false, Sphere Radius=5, UseOwnerScale=true)
   /SetVariables(Particles.Position=&Add_Vector(A=$Particles.ShapeLocation.ShapeVector, B=$Output.FN_HermiteSpline.Position)) {isSetParameters=true}
  ]
  particleUpdateScript [
   &ParticleState()
   &GravityForce(Gravity=<0 0 -1960>)
   &Drag(Drag=&RandomRangeFloat(Minimum=1, Maximum=2.5), Use Linear Drag=true, Use Rotational Drag=true)
   &SolveForcesAndVelocity(Clamp Velocity=false, Limit Acceleration=false)
   &SpriteRotationRate(Rotation Rate=10)
   &ScaleSpriteSize(Uniform Curve Sprite Scale=DataInterface() {propertyValues {Curve {keys [{interpMode=RCIM_Cubic, tangentMode=RCTM_User, time=0, value=1.2}, {interpMode=RCIM_Cubic, tangentMode=RCTM_User, time=1, value=2}]}}})
   &DynamicMaterialParameters(
    Index 0 Param 1=&RandomRangeFloat(Minimum=0, Maximum=1)
    Module.Index0Param0Write=true
    Module.Index0Param1Write=true
    Module.Index0Param2Write=true
    Module.Index0Param3Write=true
   )
   &ScaleColor(
    Scale Alpha=&Multiply_Float(
     A=&FloatFromCurve(
      FloatCurve=DataInterface() {propertyValues {Curve {keys [
          {arriveTangent=10.2356, interpMode=RCIM_Cubic, leaveTangent=10.2356, tangentMode=RCTM_User, time=0, value=0}
          <at 0.3 1 RCIM_Cubic>
          {arriveTangent=-3.6118, tangentMode=RCTM_User, time=1, value=0}
         ]}}}
      Scale Curve=&RandomRangeFloat(Minimum=0.5, Maximum=1.5)
     )
     B=1
    )
    Scale RGB=&MakeVectorFromLinearColorRGB(Color=&ColorFromCurve(ColorCurve=DataInterface() {propertyValues {
        AlphaCurve {keys [<at 0 1 RCIM_Cubic>]}
        BlueCurve {keys [<at 0 7 RCIM_Cubic>,<at 0.4104 1>,<at 0.8584 0.02 RCIM_Cubic>]}
        GreenCurve {keys [<at 0 7 RCIM_Cubic>,<at 0.4104 1>,<at 0.8584 0.02 RCIM_Cubic>]}
        RedCurve {keys [<at 0 7 RCIM_Cubic>,<at 0.4104 1>,<at 0.8584 0.02 RCIM_Cubic>]}
       }}))
    ScaleA=true
    ScaleRGB=true
   )
   &SubUVAnimation(
    End Frame Range Override=3
    Evaluation Type=::NewEnumerator0
    "Sprite Renderer "=DataInterface() {propertyValues {SpriteRenderer=&self:Smoke_0.NiagaraSpriteRendererProperties_0}}
    SubUV Animation Mode:ENiagara_SubUVLookupModeV2=::NewEnumerator2
    UseEndFrame=true
    UseStartFrame=true
   )
  ]
  rendererClasses [&NiagaraSpriteRendererProperties]
  renderers [{rendererClass=&NiagaraSpriteRendererProperties}]
 }
 Sparks {
  emitterUpdateScript [
   &EmitterState()
   &SpawnRate(SpawnRate=50, Use Spawn Probability=false)
  ]
  particleSpawnScript [
   &InitializeParticle(
    Color Mode=::NewEnumerator0
    Lifetime Mode=::NewEnumerator1
    Sprite Size Max=<1 1>
    Sprite Size Min=<0.5 0.5>
    Sprite Size Mode=::NewEnumerator2
    UsePositionOffset=false
    Write Material Random=true
   )
   &ShapeLocation(Apply To Particle Position=::NewEnumerator1, MaskSpawnGroup=false, Override Seed=false, Sphere Radius=750, UseOwnerScale=true)
   &SkeletalMeshLocation(
    Bone / Socket Index=&MakeCustomIntFromBool(Bool=$User.LeftHand, True INT=1, False INT=0)
    Bone Sampling Mode=::NewEnumerator4
    MaskSpawnGroup=false
    Orientation Sampling=::NewEnumerator1
    Skeletal Mesh=$System.Skeletal Mesh
    Write To Particle Position=false
   )
   &CameraOffset(Camera Offset Amount=200)
  ]
  particleUpdateScript [
   &ParticleState()
   /SetVariables(Particles.LocationOnSpline:NiagaraFloat=&RemapRange(
     Input Value To Remap=&Frac_Float(Float=&Add_Float(A=$Particles.NormalizedAge, B=&RandomRangeFloat(Minimum=-0.1, Maximum=0.1)))
     Input Min=0
     Input Max=1
     Output Min=$User.Erosion
     Output Max=$User.Growth
    )) {isSetParameters=true, module=SetVariables_#2}
   /SetVariables(Particles.LocationOnSpline:NiagaraFloat=&RemapRange(
     Input Value To Remap=&Frac_Float(Float=&Add_Float(A=$Particles.NormalizedAge, B=&RandomRangeFloat(Minimum=-0.1, Maximum=0.1)))
     Input Min=0
     Input Max=1
     Output Min=$User.Erosion
     Output Max=&ClampFloat(Float=&Subtract_Float(
       A=&RemapRange(Input Value To Remap=$User.Growth, Input Min=0, Input Max=1, Output Min=0, Output Max=2)
       B=&ReturnNormalizedExecIndex()
      ))
    )) {enabled=false, isSetParameters=true}
   &FN_HermiteSpline(
    Compute Velocity=true
    Direction End:Vector3f=&MakeCustomVectorFromBool(
     Bool=$User.LeftHand
     True Vector=&MakeCustomVectorFromBool(
      Bool=$User.LeftHand
      True Vector=&MakeVector(
       X=0
       Y=0
       Z=&Multiply_Float(
        A=16000
        B=&FloatFromCurve(
         FloatCurve=DataInterface() {propertyValues {Curve {keys [
             {interpMode=RCIM_Cubic, tangentMode=RCTM_User, time=0.5, value=1}
             {arriveTangent=-5.1701, interpMode=RCIM_Cubic, leaveTangent=-5.1701, tangentMode=RCTM_User, time=1, value=0}
            ]}}}
         CurveIndex=$Particles.LocationOnSpline
        )
       )
      )
      False Vector=<0 0 0>
     )
     False Vector=<0 0 -4500>
    )
    Direction Start:Vector3f=&MakeCustomVectorFromBool(
     Bool=$User.LeftHand
     True Vector=&MakeVector(
      X=&Multiply_Float(
       A=80000
       B=&FloatFromCurve(
        FloatCurve=DataInterface() {propertyValues {Curve {keys [{interpMode=RCIM_Cubic, tangentMode=RCTM_User, time=0.2, value=0}, {interpMode=RCIM_Cubic, tangentMode=RCTM_User, time=1, value=1}]}}}
        CurveIndex=$Particles.LocationOnSpline
       )
      )
      Y=16000
      Z=8000
     )
     False Vector=<0 0 43750>
    )
    Position Blend:NiagaraFloat=$Particles.LocationOnSpline
    Position End:Vector3f=&Add_Vector(
     A=$Particles.SkeletalMeshLocation.SampledPosition
     B=&RotateVectorByQuaternion(
      Vector To Rotate=&MakeCustomVectorFromBool(Bool=$User.LeftHand, True Vector=<-900 350 0>, False Vector=<1000 -350 0>)
      Quaternion=$Particles.SkeletalMeshLocation.SampledRotation
     )
    )
    Position Start:Vector3f=<-57600 42800 11500>
    Write Particles Velocity=true
   )
   &ScaleColor(
    Scale RGB=&MakeVectorFromLinearColorRGB(Color=&ColorFromCurve(ColorCurve=DataInterface() {propertyValues {
        AlphaCurve {keys [<at 0 1 RCIM_Cubic>]}
        BlueCurve {keys [<at 0.0664 0 RCIM_Cubic>,<at 0.4139 0>,<at 0.9415 0 RCIM_Cubic>]}
        GreenCurve {keys [<at 0.0664 1.5738 RCIM_Cubic>,<at 0.4139 0.4783>,<at 0.9415 0 RCIM_Cubic>]}
        RedCurve {keys [<at 0.0664 5 RCIM_Cubic>,<at 0.4139 2>,<at 0.9415 1 RCIM_Cubic>]}
       }}))
    ScaleA=true
    ScaleRGB=true
   )
   &ScaleSpriteSize(
    Uniform Curve Scale=&FloatFromCurve(
     FloatCurve=DataInterface() {propertyValues {Curve {keys [
         {arriveTangent=1.1179, time=0.1, value=1}
         {arriveTangent=0.0224, interpMode=RCIM_Cubic, leaveTangent=0.0224, tangentMode=RCTM_User, time=0.9, value=1}
         <at 1 0 RCIM_Cubic>
        ]}}}
     CurveIndex=$Particles.LocationOnSpline
     Scale Curve=110
    )
    Uniform Curve Sprite Scale=DataInterface() {propertyValues {Curve {keys [
        {arriveTangent=17.948, interpMode=RCIM_Cubic, leaveTangent=17.948, tangentMode=RCTM_User, time=0, value=0}
        {arriveTangent=1.7917, leaveTangent=1.7917, time=0.2, value=1}
        {interpMode=RCIM_Cubic, tangentMode=RCTM_User, time=0.6, value=1}
        {tangentMode=RCTM_User, time=1, value=0}
       ]}}}
   )
   /SetVariables(
    Particles.Position=&AddVectorToPosition(Position=$Particles.ShapeLocation.ShapeVector, Vector=$Output.FN_HermiteSpline.Position)
    Particles.Velocity=&Multiply_VectorByFloat(
     Vector=$Particles.Velocity
     Float=&FloatFromCurve(
      FloatCurve=DataInterface() {propertyValues {Curve {keys [{interpMode=RCIM_Cubic, tangentMode=RCTM_User, time=0.5, value=1}, {arriveTangent=-3.2842, tangentMode=RCTM_User, time=1, value=0.5}]}}}
      CurveIndex=$Particles.LocationOnSpline
     )
    )
   ) {isSetParameters=true}
   &ScaleSpriteSizeBySpeed(Max Scale Factor=<0.75 2.5>, Velocity Threshold=31250)
  ]
  rendererClasses [&NiagaraSpriteRendererProperties]
  renderers [{rendererClass=&NiagaraSpriteRendererProperties}]
 }
 Trail {
  emitterUpdateScript [
   &EmitterState()
   &SpawnBurst_Instantaneous(Spawn Count=40, Use Spawn Probability=false, UseLoopCountLimit=true)
  ]
  particleSpawnScript [&InitializeParticle(
    Color=<rgba 6.7435 2.3958 10 0.2>
    Ribbon Width=1000
    Ribbon Width Mode=::NewEnumerator1
    UsePositionOffset=false
    Write Material Random=true
   )]
  particleUpdateScript [
   &ParticleState(Kill Particles When Lifetime Has Elapsed=false, Let Infinitely Lived Particles Die When Emitter Deactivates=true)
   &SkeletalMeshLocation(
    Bone / Socket Index=&MakeCustomIntFromBool(Bool=$User.LeftHand, True INT=1, False INT=0)
    Bone Sampling Mode=::NewEnumerator4
    MaskSpawnGroup=false
    Orientation Sampling=::NewEnumerator1
    Skeletal Mesh=$System.Skeletal Mesh
    Write To Particle Position=false
   )
   &FN_HermiteSpline(
    Direction End:Vector3f=&MakeCustomVectorFromBool(
     Bool=$User.LeftHand
     True Vector=&MakeVector(
      X=0
      Y=0
      Z=&Multiply_Float(
       A=16000
       B=&FloatFromCurve(
        FloatCurve=DataInterface() {propertyValues {Curve {keys [
            {interpMode=RCIM_Cubic, tangentMode=RCTM_User, time=0.5, value=1}
            {arriveTangent=-5.1701, interpMode=RCIM_Cubic, leaveTangent=-5.1701, tangentMode=RCTM_User, time=1, value=0}
           ]}}}
        CurveIndex=&ReturnNormalizedExecIndex()
       )
      )
     )
     False Vector=<0 0 -4500>
    )
    Direction Start:Vector3f=&MakeCustomVectorFromBool(
     Bool=$User.LeftHand
     True Vector=&MakeVector(
      X=&Multiply_Float(
       A=80000
       B=&FloatFromCurve(
        FloatCurve=DataInterface() {propertyValues {Curve {keys [{interpMode=RCIM_Cubic, tangentMode=RCTM_User, time=0.2, value=0}, {interpMode=RCIM_Cubic, tangentMode=RCTM_User, time=1, value=1}]}}}
        CurveIndex=&ReturnNormalizedExecIndex()
       )
      )
      Y=16000
      Z=8000
     )
     False Vector=<0 0 43750>
    )
    Position Blend:NiagaraFloat=&ReturnNormalizedExecIndex()
    Position End:Vector3f=&Add_Vector(
     A=$Particles.SkeletalMeshLocation.SampledPosition
     B=&RotateVectorByQuaternion(
      Vector To Rotate=&MakeCustomVectorFromBool(Bool=$User.LeftHand, True Vector=<-900 350 0>, False Vector=<1000 -350 0>)
      Quaternion=$Particles.SkeletalMeshLocation.SampledRotation
     )
    )
    Position Start:Vector3f=<-57600 42800 11500>
    Write Particles Position=true
   )
   &DynamicMaterialParameters(
    Index 0 Param 1=0.3
    Index 0 Param 3=2.5
    Index 0 Param 4=2
    Index 1 Param 1=50
    Module.Index0Param0Write=true
    Module.Index0Param1Write=true
    Module.Index0Param2Write=true
    Module.Index0Param3Write=true
    Module.Index1Param0Write=true
    Module.Index1Param1Write=true
    Module.Index1Param2Write=true
    Module.Index1Param3Write=true
    Write Parameter Index 1=true
   )
   &ScaleColor(
    Scale Alpha=&FloatFromCurve(
     FloatCurve=DataInterface() {propertyValues {Curve {keys [
         {interpMode=RCIM_Cubic, tangentMode=RCTM_User, time=0, value=0}
         {interpMode=RCIM_Cubic, tangentMode=RCTM_User, time=0.5, value=1}
         {tangentMode=RCTM_User, time=1, value=0}
        ]}}}
     CurveIndex=&Frac_Float(Float=&Add_Float(
       A=&Multiply_Float(A=$Particles.NormalizedAge, B=1)
       B=&Multiply_Float(A=&Lerp_Float(A=1, B=0, Alpha=&ReturnNormalizedExecIndex()), B=6)
      ))
    )
    Scale RGB=&VectorFromFloat(Value=&FloatFromCurve(
      FloatCurve=DataInterface() {propertyValues {Curve {keys [
          {interpMode=RCIM_Cubic, tangentMode=RCTM_User, time=0, value=1}
          {interpMode=RCIM_Cubic, tangentMode=RCTM_User, time=0.5, value=2}
          {tangentMode=RCTM_User, time=1, value=1}
         ]}}}
      CurveIndex=&Frac_Float(Float=&Add_Float(
        A=&Multiply_Float(A=$Particles.NormalizedAge, B=1)
        B=&Multiply_Float(A=&Lerp_Float(A=1, B=0, Alpha=&ReturnNormalizedExecIndex()), B=2)
       ))
     ))
    ScaleA=true
    ScaleRGB=true
   )
   &ScaleColor(Scale RGB=&MakeVectorFromLinearColorRGB(Color=<rgba 1 0.6838 0.1302 1>), ScaleA=true, ScaleRGB=true) {module=ScaleColor001}
   &ScaleRibbonWidth(Ribbon Width Scale=&FloatFromCurve(
     FloatCurve=DataInterface() {propertyValues {Curve {keys [
         {arriveTangent=5.7924, interpMode=RCIM_Cubic, leaveTangent=5.7924, tangentMode=RCTM_User, time=0, value=0.4}
         {tangentMode=RCTM_User, time=0.3, value=1}
         {arriveTangent=-0.0405, interpMode=RCIM_Cubic, leaveTangent=-0.0405, tangentMode=RCTM_User, time=0.7, value=1}
        ]}}}
     CurveIndex=&ReturnNormalizedExecIndex()
     Scale Curve=&FloatFromCurve(
      FloatCurve=DataInterface() {propertyValues {Curve {keys [
          {interpMode=RCIM_Cubic, tangentMode=RCTM_User, time=0, value=0.7}
          {interpMode=RCIM_Cubic, tangentMode=RCTM_User, time=0.5, value=1.1}
          {tangentMode=RCTM_User, time=1, value=0.7}
         ]}}}
      CurveIndex=&Frac_Float(Float=&Add_Float(
        A=&Multiply_Float(A=$Particles.NormalizedAge, B=2)
        B=&Multiply_Float(A=&Lerp_Float(A=1, B=0, Alpha=&ReturnNormalizedExecIndex()), B=2)
       ))
      Scale Curve=&Lerp_Float(
       A=&FloatFromCurve(
        FloatCurve=DataInterface() {propertyValues {Curve {keys [{tangentMode=RCTM_User, time=0, value=0}, {tangentMode=RCTM_User, time=1, value=1}]}}}
        CurveIndex=&ClampFloat(Float=&Subtract_Float(
          A=&RemapRange(Input Value To Remap=$User.Growth, Input Min=0, Input Max=1, Output Min=0, Output Max=2)
          B=&ReturnNormalizedExecIndex()
         ))
       )
       B=&FloatFromCurve(
        FloatCurve=DataInterface() {propertyValues {Curve {keys [{tangentMode=RCTM_User, time=0, value=1}, {tangentMode=RCTM_User, time=1, value=0}]}}}
        CurveIndex=&ClampFloat(Float=&Subtract_Float(
          A=&RemapRange(Input Value To Remap=$User.Erosion, Input Min=0, Input Max=1, Output Min=0, Output Max=2)
          B=&ReturnNormalizedExecIndex()
         ))
       )
       Alpha=$User.Flip
      )
     )
    ))
  ]
  rendererClasses [&NiagaraRibbonRendererProperties]
  renderers [{rendererClass=&NiagaraRibbonRendererProperties}]
 }
}
summary {
 systemName=NS_GrillMoon_Demon_AbsorbTrail
 emitters [
  {emitterName=MainTrail, rendererClasses [&NiagaraRibbonRendererProperties, &NiagaraSpriteRendererProperties]}
  {emitterName=Sparks, rendererClasses [&NiagaraSpriteRendererProperties]}
  {bEnabled=false, emitterName=LeadParticle, rendererClasses [&NiagaraSpriteRendererProperties]}
  {bEnabled=false, emitterName=AccentRibbons, rendererClasses [&NiagaraRibbonRendererProperties]}
  {emitterName=Smoke, rendererClasses [&NiagaraSpriteRendererProperties]}
  {emitterName=ExitPortal, rendererClasses [&NiagaraMeshRendererProperties]}
  {emitterName=Trail, rendererClasses [&NiagaraRibbonRendererProperties]}
 ]
 userVariables [
  User.Emitter_TipFade {type=NiagaraFloat, defaultValue=2, description=""}
  User.Erosion {type=NiagaraFloat, defaultValue=0, description=""}
  User.Flip {type=NiagaraFloat, defaultValue=0, description=""}
  User.Growth {type=NiagaraFloat, defaultValue=1, description=""}
  User.LeftHand {type=NiagaraBool, defaultValue=true}
  User.SpawnRate {type=NiagaraFloat, defaultValue=0, description=""}
  User.True Vector_Direction End {type=Vector3f, description="", defaultValue {struct=Vector3f, value=<-329812.6875 -90049.1797 -104875.8203>}}
 ]
}
